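# Audio Level Sequence for the tac1000 shot of PistolCanary, scoring the tentacle boss with overlapping sounds at 60 fps.
# A master TentacleAudio track layers five sections at constant default volume and pitch.
# The AudioManager binding fires four event triggers shortly after four of those sections start, and toggles its Print bool on, then off, during the shot.
# TentacleBoss is bound but has no tracks here.
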LevelSequence /PistolCanary/Sequence/tac/tac1000/PistolCanary_tac1000_aud
bindings [
 PistolCanary_AudioManager {
  id=#1
  possessed_class=PistolCanary_AudioManager_C
  tracks [
   {
    type=MovieSceneEventTrack
    sections [{
      type=MovieSceneEventTriggerSection
      channels [None {
        type=MovieSceneScriptingEventChannel
        keys [
         [3615.3425, "<Struct 'MovieSceneEvent' (0x852ff394060) {}>"]
         [3783.4375, "<Struct 'MovieSceneEvent' (0x852ff393e80) {}>"]
         [3866.8975, "<Struct 'MovieSceneEvent' (0x852ff394380) {}>"]
         [3956, "<Struct 'MovieSceneEvent' (0x852ff394560) {}>"]
        ]
       }]
     }]
   }
   {
    type=MovieSceneBoolTrack
    display_name=Print
    sections [{type=MovieSceneBoolSection, channels [None {type=MovieSceneScriptingBoolChannel, default=true, keys [[3633,true],[4012,false]]}]}]
   }
  ]
 }
 TentacleBoss {id=#2, possessed_class=PistolCanary_TentacleBoss_C, tracks []}
]
display_rate {denominator=1, numerator=60}
tick_resolution {denominator=1, numerator=24000}
tracks [{
  type=MovieSceneAudioTrack
  display_name=TentacleAudio
  sections [
   {
    type=MovieSceneAudioSection
    end_frame=4040
    start_frame=3857
    channels [
     Volume {type=MovieSceneScriptingFloatChannel, default=1}
     Pitch {type=MovieSceneScriptingFloatChannel, default=1}
     AttachActor {type=MovieSceneScriptingActorReferenceChannel}
    ]
   }
   {
    type=MovieSceneAudioSection
    end_frame=4055
    start_frame=3599
    channels [
     Volume {type=MovieSceneScriptingFloatChannel, default=1}
     Pitch {type=MovieSceneScriptingFloatChannel, default=1}
     AttachActor {type=MovieSceneScriptingActorReferenceChannel}
    ]
   }
   {
    type=MovieSceneAudioSection
    end_frame=4130
    start_frame=3778
    channels [
     Volume {type=MovieSceneScriptingFloatChannel, default=1}
     Pitch {type=MovieSceneScriptingFloatChannel, default=1}
     AttachActor {type=MovieSceneScriptingActorReferenceChannel}
    ]
   }
   {
    type=MovieSceneAudioSection
    end_frame=4227
    start_frame=3952
    channels [
     Volume {type=MovieSceneScriptingFloatChannel, default=1}
     Pitch {type=MovieSceneScriptingFloatChannel, default=1}
     AttachActor {type=MovieSceneScriptingActorReferenceChannel}
    ]
   }
   {
    type=MovieSceneAudioSection
    end_frame=4102
    start_frame=4017
    channels [
     Volume {type=MovieSceneScriptingFloatChannel, default=1}
     Pitch {type=MovieSceneScriptingFloatChannel, default=1}
     AttachActor {type=MovieSceneScriptingActorReferenceChannel}
    ]
   }
  ]
 }]
