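# Gameplay-event Level Sequence for shot tac1000 in PistolCanary that fires scripted events on the TentacleBoss.
# Two possessable TentacleBoss bindings, labelled for Hermes_Terrain and Test Level, each carry an event trigger track.
# Both tracks key three events at identical, closely spaced times near the start of the sequence.
# The sequence has no other tracks, so it only triggers events and holds no animation or camera data.
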LevelSequence /PistolCanary/Sequence/tac/tac1000/PistolCanary_tac1000_gam
bindings [
 TentacleBoss - Hermes_Terrain {
  id=#1
  possessed_class=PistolCanary_TentacleBoss_C
  tracks [{
    type=MovieSceneEventTrack
    sections [{
      type=MovieSceneEventTriggerSection
      channels [None {
        type=MovieSceneScriptingEventChannel
        keys [
         [3800, "<Struct 'MovieSceneEvent' (0x852ff394ce0) {}>"]
         [3888, "<Struct 'MovieSceneEvent' (0x852ff394060) {}>"]
         [3975, "<Struct 'MovieSceneEvent' (0x852ff393e80) {}>"]
        ]
       }]
     }]
   }]
 }
 TentacleBoss - Test Level {
  id=#2
  possessed_class=PistolCanary_TentacleBoss_C
  tracks [{
    type=MovieSceneEventTrack
    sections [{
      type=MovieSceneEventTriggerSection
      channels [None {
        type=MovieSceneScriptingEventChannel
        keys [
         [3800, "<Struct 'MovieSceneEvent' (0x852ff394600) {}>"]
         [3888, "<Struct 'MovieSceneEvent' (0x852ff394560) {}>"]
         [3975, "<Struct 'MovieSceneEvent' (0x852ff394380) {}>"]
        ]
       }]
     }]
   }]
 }
]
display_rate {denominator=1, numerator=60}
tick_resolution {denominator=1, numerator=24000}
tracks []
