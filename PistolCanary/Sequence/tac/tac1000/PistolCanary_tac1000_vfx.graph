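# VFX Level Sequence for shot tac1000, choreographing a PistolCanary tentacle boss charge-up followed by three laser blasts.
# An attached charge-up Niagara system ramps progress, lightning and post-process parameters, then three attached laser systems appear in staggered succession with animated beam width and transforms.
# Body material pulses and glow on two TentacleBoss bindings, plus post-process volume intensity spikes, are keyed to the shots.
# Both boss bindings also carry event tracks.
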LevelSequence /PistolCanary/Sequence/tac/tac1000/PistolCanary_tac1000_vfx
bindings [
 NiagaraComponent0 {
  id=#1
  parent_id=#2
  possessed_class=NiagaraComponent
  tracks [
   {type=MovieSceneNiagaraSystemTrack, sections [{type=MovieSceneNiagaraSystemSpawnSection, end_frame=4486, start_frame=3886}]}
   {
    type=MovieSceneNiagaraFloatParameterTrack
    display_name=BeamWidth
    sections [{type=MovieSceneFloatSection, channels [None {type=MovieSceneScriptingFloatChannel, default=1, const [1,2]}]}]
   }
   {
    type=MovieSceneParticleTrack
    display_name=FX System
    sections [{
      type=MovieSceneParticleSection
      channels [None {type=MovieSceneScriptingParticleChannel, keys [[3886, "<ParticleKey.ACTIVATE: 0>"], [3932, "<ParticleKey.DEACTIVATE: 1>"]]}]
     }]
   }
  ]
 }
 PPV_PistolCanary_VFX {
  id=#3
  possessed_class=PostProcessVolume
  tracks [
   {
    type=MovieSceneFloatTrack
    display_name=Blend Weight
    sections [{type=MovieSceneFloatSection, channels [None {type=MovieSceneScriptingFloatChannel, default=1, keys [[3614,0],[3720,1],[3999,1],[4026,0]]}]}]
   }
   {
    type=MovieSceneFloatTrack
    display_name="Intensity (Settings)"
    sections [{
      type=MovieSceneFloatSection
      channels [None {
        type=MovieSceneScriptingFloatChannel
        default=1.5
        keys [[3799,1.5],[3801,2.5],[3833,1.5],[3886,1.5],[3888,2.5],[3920,1.5],[3972,1.5],[3974,2.5],[4006,1.5]]
       }]
     }]
   }
  ]
 }
 BrushComponent0 {id=#4, parent_id=#3, possessed_class=BrushComponent, tracks []}
 SkeletalMesh {
  id=#5
  parent_id=#6
  possessed_class=SkeletalMeshComponent
  tracks [
   {
    type=MovieSceneComponentMaterialTrack
    display_name=Material Slot: Body_A
    sections [{
      type=MovieSceneComponentMaterialParameterSection
      end_frame=4118
      start_frame=3620
      channels [
       GlowPulseEXP {type=MovieSceneScriptingFloatChannel, const [0.015,1]}
       PulseSpeed {type=MovieSceneScriptingFloatChannel, default=0, const [0,1]}
       PulseManualOffset {
        type=MovieSceneScriptingFloatChannel
        default=0
        keys [[3632,0],[3740,1],[3740.56,0],[3793,0],[3802,1],[3803,0],[3880,0],[3889,1],[3890,0],[3961,0],[3970,1],[3971,0]]
       }
       PulseRadius {type=MovieSceneScriptingFloatChannel, const [0.33,1]}
       EmissiveGlowColorR {type=MovieSceneScriptingFloatChannel, const [12,1]}
       EmissiveGlowColorG {type=MovieSceneScriptingFloatChannel, const [2.2855,1]}
       EmissiveGlowColorB {type=MovieSceneScriptingFloatChannel, default=100, const [100,1]}
       EmissiveGlowColorA {type=MovieSceneScriptingFloatChannel, const [1,1]}
      ]
     }]
   }
   {
    type=MovieSceneComponentMaterialTrack
    display_name=Material Slot: Body_B
    sections [{
      type=MovieSceneComponentMaterialParameterSection
      end_frame=4118
      start_frame=3620
      channels [
       PulseManualOffset {
        type=MovieSceneScriptingFloatChannel
        default=0
        keys [[3632,0],[3740,1],[3740.56,0],[3793,0],[3802,1],[3803,0],[3880,0],[3889,1],[3890,0],[3961,0],[3970,1],[3971,0]]
       }
       PulseSpeed {type=MovieSceneScriptingFloatChannel, default=0, const [0,1]}
       EmissiveGlowColorR {type=MovieSceneScriptingFloatChannel, const [12,1]}
       EmissiveGlowColorG {type=MovieSceneScriptingFloatChannel, const [2.2855,1]}
       EmissiveGlowColorB {type=MovieSceneScriptingFloatChannel, default=100, const [100,1]}
       EmissiveGlowColorA {type=MovieSceneScriptingFloatChannel, const [1,1]}
      ]
     }]
   }
   {
    type=MovieSceneComponentMaterialTrack
    display_name=Material Slot: Body_C
    sections [{
      type=MovieSceneComponentMaterialParameterSection
      end_frame=4118
      start_frame=3620
      channels [
       PulseSpeed {type=MovieSceneScriptingFloatChannel, default=0, const [0,1]}
       PulseManualOffset {
        type=MovieSceneScriptingFloatChannel
        default=0
        keys [[3632,0],[3740,1],[3740.56,0],[3793,0],[3802,1],[3803,0],[3880,0],[3889,1],[3890,0],[3961,0],[3970,1],[3971,0]]
       }
       EmissiveGlowColorR {type=MovieSceneScriptingFloatChannel, const [12,1]}
       EmissiveGlowColorG {type=MovieSceneScriptingFloatChannel, const [2.2855,1]}
       EmissiveGlowColorB {type=MovieSceneScriptingFloatChannel, default=100, const [100,1]}
       EmissiveGlowColorA {type=MovieSceneScriptingFloatChannel, const [1,1]}
      ]
     }]
   }
   {
    type=MovieSceneComponentMaterialTrack
    display_name=Material Slot: Body_D
    sections [{
      type=MovieSceneComponentMaterialParameterSection
      end_frame=4118
      start_frame=3620
      channels [
       PulseManualOffset {
        type=MovieSceneScriptingFloatChannel
        default=0
        keys [[3632,0],[3740,1],[3740.56,0],[3793,0],[3802,1],[3803,0],[3880,0],[3889,1],[3890,0],[3961,0],[3970,1],[3971,0]]
       }
       PulseSpeed {type=MovieSceneScriptingFloatChannel, default=0, const [0,2]}
       EmissiveGlowColorR {type=MovieSceneScriptingFloatChannel, const [12,1]}
       EmissiveGlowColorG {type=MovieSceneScriptingFloatChannel, const [2.2855,1]}
       EmissiveGlowColorB {type=MovieSceneScriptingFloatChannel, default=100, const [100,1]}
       EmissiveGlowColorA {type=MovieSceneScriptingFloatChannel, const [1,1]}
      ]
     }]
   }
  ]
 }
 TentacleBoss {
  id=#7
  possessed_class=PistolCanary_TentacleBoss_C
  tracks [{
    type=MovieSceneEventTrack
    sections [{
      type=MovieSceneEventTriggerSection
      channels [None {type=MovieSceneScriptingEventChannel, const ["<Struct 'MovieSceneEvent' (0x852ff396860) {}>", 1]}]
     }]
   }]
 }
 TentacleBoss {
  id=#6
  possessed_class=PistolCanary_TentacleBoss_C
  tracks [{
    type=MovieSceneEventTrack
    sections [{
      type=MovieSceneEventTriggerSection
      channels [None {type=MovieSceneScriptingEventChannel, const ["<Struct 'MovieSceneEvent' (0x852ff396a40) {}>", 1]}]
     }]
   }]
 }
 NS_TentacleAttackLaser3 {
  id=#8
  possessed_class=NiagaraActor
  tracks [
   {type=MovieScene3DAttachTrack, sections [{type=MovieScene3DAttachSection, end_frame=4199, start_frame=3767}]}
   {
    type=MovieSceneSpawnTrack
    sections [{type=MovieSceneSpawnSection, channels [None {type=MovieSceneScriptingBoolChannel, default=true, keys [[3773,false],[3969,true]]}]}]
   }
   {
    type=MovieScene3DTransformTrack
    sections [{
      type=MovieScene3DTransformSection
      channels [
       Location.X {type=MovieSceneScriptingDoubleChannel, default=0}
       Location.Y {type=MovieSceneScriptingDoubleChannel, default=0}
       Location.Z {type=MovieSceneScriptingDoubleChannel, default=0}
       Rotation.X {type=MovieSceneScriptingDoubleChannel, default=179.0786, keys [[3970,111.8465],[3974,111.2328]]}
       Rotation.Y {type=MovieSceneScriptingDoubleChannel, default=-71, keys [[3970,-82.3252],[3974,-77.1298]]}
       Rotation.Z {type=MovieSceneScriptingDoubleChannel, default=-182.8809, const [-117.9854,2]}
       Scale.X {type=MovieSceneScriptingDoubleChannel, default=15}
       Scale.Y {type=MovieSceneScriptingDoubleChannel, default=15}
       Scale.Z {type=MovieSceneScriptingDoubleChannel, default=25, keys [[3969,15],[3986,25]]}
      ]
     }]
   }
  ]
 }
 NiagaraComponent0 {
  id=#9
  parent_id=#8
  possessed_class=NiagaraComponent
  tracks [
   {type=MovieSceneNiagaraSystemTrack, sections [{type=MovieSceneNiagaraSystemSpawnSection, end_frame=4569, start_frame=3969}]}
   {
    type=MovieSceneNiagaraFloatParameterTrack
    display_name=BeamWidth
    sections [{type=MovieSceneFloatSection, channels [None {type=MovieSceneScriptingFloatChannel, default=1, keys [[3910,1],[3917,0],[3962,1]]}]}]
   }
   {
    type=MovieSceneParticleTrack
    display_name=FX System
    sections [{type=MovieSceneParticleSection, channels [None {type=MovieSceneScriptingParticleChannel}]}]
   }
  ]
 }
 NS_TentacleAttackLaser {
  id=#10
  possessed_class=NiagaraActor
  tracks [
   {type=MovieScene3DAttachTrack, sections [{type=MovieScene3DAttachSection, end_frame=3978, start_frame=3767}]}
   {
    type=MovieSceneSpawnTrack
    sections [{
      type=MovieSceneSpawnSection
      channels [None {type=MovieSceneScriptingBoolChannel, default=true, keys [[3773,false],[3799,true],[3874,false]]}]
     }]
   }
   {
    type=MovieScene3DTransformTrack
    sections [{
      type=MovieScene3DTransformSection
      channels [
       Location.X {type=MovieSceneScriptingDoubleChannel, default=0}
       Location.Y {type=MovieSceneScriptingDoubleChannel, default=0}
       Location.Z {type=MovieSceneScriptingDoubleChannel, default=0}
       Rotation.X {type=MovieSceneScriptingDoubleChannel, default=179.0786, const [-87.2302,1]}
       Rotation.Y {type=MovieSceneScriptingDoubleChannel, default=-56.8563, const [-22.8886,1]}
       Rotation.Z {type=MovieSceneScriptingDoubleChannel, default=-182.8809, const [87.9642,1]}
       Scale.X {type=MovieSceneScriptingDoubleChannel, default=15, const [15,1]}
       Scale.Y {type=MovieSceneScriptingDoubleChannel, default=15, const [15,1]}
       Scale.Z {type=MovieSceneScriptingDoubleChannel, default=30, const [30,1]}
      ]
     }]
   }
  ]
 }
 NiagaraComponent0 {
  id=#11
  parent_id=#12
  possessed_class=NiagaraComponent
  tracks [
   {type=MovieSceneNiagaraSystemTrack, sections [{type=MovieSceneNiagaraSystemSpawnSection, end_frame=4230, start_frame=3630}]}
   {
    type=MovieSceneNiagaraVectorParameterTrack
    display_name=ChargeScaleVec
    sections [{
      type=MovieSceneFloatVectorSection
      channels [
       Vector.X {type=MovieSceneScriptingFloatChannel, default=0.6, keys [[3632,0],[3637,0.6],[3727,1]]}
       Vector.Y {type=MovieSceneScriptingFloatChannel, default=1, const [1,2]}
       Vector.Z {type=MovieSceneScriptingFloatChannel, default=1, const [1,2]}
      ]
     }]
   }
   {
    type=MovieSceneNiagaraFloatParameterTrack
    display_name=ChargeProgress
    sections [{type=MovieSceneFloatSection, channels [None {type=MovieSceneScriptingFloatChannel, default=1, keys [[3631,0],[3663,0.3],[3722,1]]}]}]
   }
   {
    type=MovieSceneNiagaraFloatParameterTrack
    display_name=PlasmaEnabled
    sections [{type=MovieSceneFloatSection, channels [None {type=MovieSceneScriptingFloatChannel, default=0, keys [[3728,1],[3737,0]]}]}]
   }
   {
    type=MovieSceneNiagaraColorParameterTrack
    display_name=ChargeScaleColor
    sections [{
      type=MovieSceneColorSection
      channels [
       Color.R {type=MovieSceneScriptingFloatChannel, default=1}
       Color.G {type=MovieSceneScriptingFloatChannel, default=1}
       Color.B {type=MovieSceneScriptingFloatChannel, default=1}
       Color.A {type=MovieSceneScriptingFloatChannel, default=0, keys [[3646,0],[3681,1],[4006,1],[4106,0]]}
      ]
     }]
   }
   {
    type=MovieSceneNiagaraFloatParameterTrack
    display_name=ChargeProgressLightning
    sections [{type=MovieSceneFloatSection, channels [None {type=MovieSceneScriptingFloatChannel, default=0.5472, keys [[3633,0],[3689,0.5472],[3785,1]]}]}]
   }
   {
    type=MovieSceneNiagaraFloatParameterTrack
    display_name=PostProcess
    sections [{
      type=MovieSceneFloatSection
      channels [None {type=MovieSceneScriptingFloatChannel, default=0, thinned_from=5, keys [[3634,0],[3814,1],[4058,1],[4110,0]]}]
     }]
   }
   {
    type=MovieSceneNiagaraFloatParameterTrack
    display_name=TipDelay
    sections [{type=MovieSceneFloatSection, channels [None {type=MovieSceneScriptingFloatChannel, default=0.8}]}]
   }
   {
    type=MovieSceneParticleTrack
    display_name=FX System
    sections [{
      type=MovieSceneParticleSection
      channels [None {type=MovieSceneScriptingParticleChannel, keys [[3630, "<ParticleKey.ACTIVATE: 0>"], [4108, "<ParticleKey.DEACTIVATE: 1>"]]}]
     }]
   }
   {
    type=MovieSceneNiagaraFloatParameterTrack
    display_name=LightningEnabled
    sections [{type=MovieSceneFloatSection, channels [None {type=MovieSceneScriptingFloatChannel, default=0, keys [[3776,1],[3784,0]]}]}]
   }
  ]
 }
 NS_TentacleAttackLaser2 {
  id=#2
  possessed_class=NiagaraActor
  tracks [
   {type=MovieScene3DAttachTrack, sections [{type=MovieScene3DAttachSection, end_frame=4421, start_frame=3767}]}
   {
    type=MovieSceneSpawnTrack
    sections [{type=MovieSceneSpawnSection, channels [None {type=MovieSceneScriptingBoolChannel, default=true, keys [[3773,false],[3886,true]]}]}]
   }
   {
    type=MovieScene3DTransformTrack
    sections [{
      type=MovieScene3DTransformSection
      channels [
       Location.X {type=MovieSceneScriptingDoubleChannel, default=0}
       Location.Y {type=MovieSceneScriptingDoubleChannel, default=0}
       Location.Z {type=MovieSceneScriptingDoubleChannel, default=0}
       Rotation.X {type=MovieSceneScriptingDoubleChannel, default=-179.0908, keys [[3887,-177.3067],[3894,-179.0908]]}
       Rotation.Y {type=MovieSceneScriptingDoubleChannel, default=-65.5282, keys [[3887,-81.9591],[3894,-65.5282]]}
       Rotation.Z {type=MovieSceneScriptingDoubleChannel, default=175.7082, keys [[3887,173.8689],[3894,175.7082]]}
       Scale.X {type=MovieSceneScriptingDoubleChannel, default=15}
       Scale.Y {type=MovieSceneScriptingDoubleChannel, default=15}
       Scale.Z {type=MovieSceneScriptingDoubleChannel, default=20, keys [[3886,10],[3889,20]]}
      ]
     }]
   }
  ]
 }
 NiagaraComponent0 {
  id=#13
  parent_id=#10
  possessed_class=NiagaraComponent
  tracks [
   {type=MovieSceneNiagaraSystemTrack, sections [{type=MovieSceneNiagaraSystemSpawnSection, end_frame=3875, start_frame=3799}]}
   {
    type=MovieSceneNiagaraFloatParameterTrack
    display_name=BeamWidth
    sections [{type=MovieSceneFloatSection, channels [None {type=MovieSceneScriptingFloatChannel, default=0, keys [[3823,1],[3830,0]]}]}]
   }
   {
    type=MovieSceneParticleTrack
    display_name=FX System
    sections [{
      type=MovieSceneParticleSection
      channels [None {type=MovieSceneScriptingParticleChannel, keys [[3799, "<ParticleKey.ACTIVATE: 0>"], [3841, "<ParticleKey.DEACTIVATE: 1>"]]}]
     }]
   }
  ]
 }
 SkeletalMesh {
  id=#14
  parent_id=#7
  possessed_class=SkeletalMeshComponent
  tracks [
   {
    type=MovieSceneComponentMaterialTrack
    display_name=Material Slot: Body_A
    sections [{
      type=MovieSceneComponentMaterialParameterSection
      end_frame=4118
      start_frame=3605
      channels [
       PulseSpeed {type=MovieSceneScriptingFloatChannel, default=0, const [0,1]}
       PulseManualOffset {
        type=MovieSceneScriptingFloatChannel
        default=0
        keys [[3632,0],[3740,1],[3740.56,0],[3793,0],[3802,1],[3803,0],[3880,0],[3889,1],[3890,0],[3961,0],[3970,1],[3971,0]]
       }
       GlowPulseEXP {type=MovieSceneScriptingFloatChannel, default=0.079, keys [[3609,0.079],[3623,0.069],[3633,0.079]]}
       EmissiveGlowColorR {type=MovieSceneScriptingFloatChannel, const [12,2]}
       EmissiveGlowColorG {type=MovieSceneScriptingFloatChannel, const [2.2855,2]}
       EmissiveGlowColorB {type=MovieSceneScriptingFloatChannel, default=974.9821, keys [[3632,1.8839],[3679,100],[4011,100],[4086,1.8839]]}
       EmissiveGlowColorA {type=MovieSceneScriptingFloatChannel, const [1,2]}
      ]
     }]
   }
   {
    type=MovieSceneComponentMaterialTrack
    display_name=Material Slot: Body_B
    sections [{
      type=MovieSceneComponentMaterialParameterSection
      end_frame=4118
      start_frame=3609
      channels [
       PulseManualOffset {
        type=MovieSceneScriptingFloatChannel
        default=0
        keys [[3632,0],[3740,1],[3740.56,0],[3793,0],[3802,1],[3803,0],[3880,0],[3889,1],[3890,0],[3961,0],[3970,1],[3971,0]]
       }
       PulseSpeed {type=MovieSceneScriptingFloatChannel, default=0, const [0,1]}
       GlowPulseEXP {type=MovieSceneScriptingFloatChannel, default=0.079, keys [[3609,0.079],[3623,0.069],[3633,0.079]]}
       EmissiveGlowColorR {type=MovieSceneScriptingFloatChannel, const [12,2]}
       EmissiveGlowColorG {type=MovieSceneScriptingFloatChannel, const [2.2855,2]}
       EmissiveGlowColorB {type=MovieSceneScriptingFloatChannel, default=1.8839, keys [[3638,1.8839],[3680,100],[4011,100],[4086,1.8839]]}
       EmissiveGlowColorA {type=MovieSceneScriptingFloatChannel, const [1,2]}
      ]
     }]
   }
   {
    type=MovieSceneComponentMaterialTrack
    display_name=Material Slot: Body_C
    sections [{
      type=MovieSceneComponentMaterialParameterSection
      end_frame=4118
      start_frame=3606
      channels [
       PulseSpeed {type=MovieSceneScriptingFloatChannel, default=0, const [0,1]}
       PulseManualOffset {
        type=MovieSceneScriptingFloatChannel
        default=0
        keys [[3632,0],[3740,1],[3740.56,0],[3793,0],[3802,1],[3803,0],[3880,0],[3889,1],[3890,0],[3961,0],[3970,1],[3971,0]]
       }
       GlowPulseEXP {type=MovieSceneScriptingFloatChannel, keys [[3608,0.079],[3622,0.069],[3632,0.079],[3853,0.079]]}
       EmissiveGlowColorR {type=MovieSceneScriptingFloatChannel, const [12,2]}
       EmissiveGlowColorG {type=MovieSceneScriptingFloatChannel, const [2.2855,2]}
       EmissiveGlowColorB {type=MovieSceneScriptingFloatChannel, default=1.8839, keys [[3644,1.8839],[3681,100],[4011,100],[4086,1.8839]]}
       EmissiveGlowColorA {type=MovieSceneScriptingFloatChannel, const [1,2]}
      ]
     }]
   }
   {
    type=MovieSceneComponentMaterialTrack
    display_name=Material Slot: Body_D
    sections [{
      type=MovieSceneComponentMaterialParameterSection
      end_frame=4118
      start_frame=3607
      channels [
       PulseManualOffset {
        type=MovieSceneScriptingFloatChannel
        default=0
        keys [[3632,0],[3740,1],[3740.56,0],[3793,0],[3802,1],[3803,0],[3880,0],[3889,1],[3890,0],[3961,0],[3970,1],[3971,0]]
       }
       PulseSpeed {type=MovieSceneScriptingFloatChannel, default=0, const [0,2]}
       GlowPulseEXP {type=MovieSceneScriptingFloatChannel, keys [[3608,0.079],[3622,0.069],[3632,0.079]]}
       EmissiveGlowColorR {type=MovieSceneScriptingFloatChannel, const [12,2]}
       EmissiveGlowColorG {type=MovieSceneScriptingFloatChannel, const [2.2855,2]}
       EmissiveGlowColorB {type=MovieSceneScriptingFloatChannel, default=1.8839, keys [[3649,1.8839],[3681,100],[4011,100],[4086,1.8839]]}
       EmissiveGlowColorA {type=MovieSceneScriptingFloatChannel, const [1,2]}
      ]
     }]
   }
  ]
 }
 NS_TentacleAttackChargeUpLaser {
  id=#12
  possessed_class=NiagaraActor
  tracks [{type=MovieScene3DAttachTrack, sections [{type=MovieScene3DAttachSection, end_frame=11296, start_frame=3608}]}]
 }
]
display_rate {denominator=1, numerator=60}
tick_resolution {denominator=1, numerator=24000}
tracks []
